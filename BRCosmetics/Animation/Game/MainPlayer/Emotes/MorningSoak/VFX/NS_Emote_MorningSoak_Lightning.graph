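# Niagara system for the MorningSoak emote that adds lightning and fire effects to points on a skeletal mesh.
# HandLightning spawns looping, oriented mesh bolts split between two bone/socket points, with curve-driven glow and material parameters.
# Small_Lightning fires brief, delayed bursts of very short-lived, jittering mesh arcs across nine other points.
# Fire re-samples one point each update, drifting SubUV flame sprites away from it, while the Flame emitter has no renderer.
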NiagaraSystem /BRCosmetics/Animation/Game/MainPlayer/Emotes/MorningSoak/VFX/NS_Emote_MorningSoak_Lightning
dependencies {
 usedDataInterfaces [
  &NiagaraDataInterfaceArrayInt32
  &NiagaraDataInterfaceCurve
  &NiagaraDataInterfaceColorCurve
  &NiagaraDataInterfaceSpriteRendererInfo
  &NiagaraDataInterfaceSkeletalMesh
 ]
 usedDynamicInputs [&MaskVectorBySpawnGroup]
 usedModules [
  &self:SystemSpawnScript.SystemScriptSource.SystemScriptGraph.NiagaraNodeAssignment_1.SetVariables_#1
  &SystemState
 ]
}
emitters {
 Fire {
  emitterUpdateScript [
   &EmitterState(Life Cycle Mode=::NewEnumerator1, Loop Behavior=::NewEnumerator1, Loop Duration Mode=::NewEnumerator1, UseLoopDelay=false)
   &SpawnRate(SpawnRate=20, Use Spawn Probability=false)
  ]
  particleSpawnScript [
   &InitializeParticle(
    Lifetime=&RandomRangeFloat(Minimum=0.3, Maximum=0.6)
    Sprite Rotation Mode=::NewEnumerator2
    Sprite Size Mode=::NewEnumerator1
    Uniform Sprite Size Max=62
    Uniform Sprite Size Min=50
    UsePositionOffset=false
    Write Material Random=true
   )
   &CameraOffset(Camera Offset Amount=5)
  ]
  particleUpdateScript [
   &ParticleState()
   &SkeletalMeshLocation(Bone / Socket Index=12, Bone Sampling Mode=::NewEnumerator4, MaskSpawnGroup=false, Skeletal Mesh=$System.Skeletal Mesh)
   &ShapeLocation(MaskSpawnGroup=false, Override Seed=false, Shape Origin=$StackContext.Position, Sphere Radius=1, UseOwnerScale=true)
   &OffsetPosition(Position Offset=&Lerp_Vector(
     A=<0 0 0>
     B=&Multiply_VectorByFloat(Vector=&RandomRangeVector(Minimum=<-80 -80 50>, Maximum=<0 50 240>), Float=0.1)
     Alpha=&FloatFromCurve(FloatCurve=DataInterface() {propertyValues {Curve {
         defaultValue=0
         keys [
          {arriveTangent=1.2528, interpMode=RCIM_Cubic, leaveTangent=1.2528, tangentMode=RCTM_User, time=0, value=0}
          {interpMode=RCIM_Cubic, tangentMode=RCTM_User, time=1, value=1}
         ]
        }}})
    ))
   &SpriteRotationRate(Rotation Rate=&RandomRangeFloat(Minimum=-300, Maximum=300))
   &ScaleSpriteSize(Uniform Curve Sprite Scale=DataInterface() {propertyValues {Curve {
       defaultValue=0.0426
       keys [<at 0 0 RCIM_Cubic>, {arriveTangent=0.1, interpMode=RCIM_Cubic, leaveTangent=0.1, time=0.1866, value=0.3551}, <at 1 0.1 RCIM_Cubic>]
      }}})
   &DynamicMaterialParameters(
    Index 0 Param 4=2.5
    Module.Index0Float0Write=true
    Module.Index0Float1Write=true
    Module.Index0Float2Write=true
    Module.Index0Float3Write=true
   )
   &GravityForce(Gravity=&RandomRangeVector(Minimum=<0 0 100>, Maximum=<0 0 150>))
   &SolveForcesAndVelocity(Clamp Velocity=false, Limit Acceleration=false)
   &ScaleColor(
    Linear Color Curve=DataInterface() {propertyValues {
      AlphaCurve {keys [<at 0.5 1>,<at 1 0.1>]}
      BlueCurve {keys [<at 0 0.511>,<at 0.6063 0.0207>]}
      GreenCurve {keys [<at 0 1.229>,<at 0.6063 0.12>]}
      RedCurve {keys [<at 0 6>,<at 0.6063 2>]}
     }}
    Scale Mode=2
   )
   &SubUVAnimation(
    End Frame Range Override=12
    "Sprite Renderer "=DataInterface() {propertyValues {SpriteRenderer=&self:Fire_0.Renderer}}
    Start Frame Range Override=8
    UseEndFrame=true
    UseStartFrame=true
   )
  ]
  rendererClasses [&NiagaraSpriteRendererProperties]
  renderers [{rendererClass=&NiagaraSpriteRendererProperties}]
 }
 Flame {rendererClasses [], renderers []}
 HandLightning {
  emitterUpdateScript [
   &EmitterState(
    Inactive Response=::NewEnumerator1
    Life Cycle Mode=::NewEnumerator1
    Loop Behavior=::NewEnumerator1
    Loop Delay=1.3076
    Loop Duration=0.834
    UseLoopDelay=true
   )
   &SpawnRate(SpawnRate=&RandomRangeFloat(Minimum=30, Maximum=36), Use Spawn Probability=false) {module=SpawnRate001}
  ]
  particleSpawnScript [
   &PartitionParticles()
   &InitializeParticle(
    Color=&Multiply_LinearColorByFloat(Linear Color=<rgba 1 0.7986 0 5>, Float=0.4)
    Lifetime Max=0.356
    Lifetime Min:NiagaraFloat=&RandomRangeFloat(Minimum=0.2, Maximum=0.32)
    Lifetime Mode=::NewEnumerator1
    Mesh Scale=<1.2 1.2 3>
    Mesh Scale Mode=::NewEnumerator4
    UsePositionOffset=false
    Write Material Random=true
   )
   &SkeletalMeshLocation(
    Bone / Socket Index=&SelectIntFromArray(Direct Array Index=$Particles.PartitionParticles.Partition, Int Selection Array=DataInterface() {propertyValues {IntData [10,11]}})
    Bone Sampling Mode=::NewEnumerator4
    MaskSpawnGroup=false
    Skeletal Mesh=$System.Skeletal Mesh
   )
   &ShapeLocation(
    Cylinder Height=5.5
    Cylinder Radius=0
    MaskSpawnGroup=false
    Override Seed=false
    Shape Origin=$Particles.SkeletalMeshLocation.SampledPosition
    Shape Primitive=::NewEnumerator1
    Surface Only=false
    UseOwnerScale=true
   ) {module=ShapeLocation001}
   &InitialMeshOrientation(
    Mesh Axis To Orient=<0 0 1>
    Mesh Orientation Mode=::NewEnumerator2
    Orientation Vector=&MaskVector(Mask Source=1, Mask Parameter=$Particles.PartitionParticles.Partition, Mask=0, Masked Vector=<0 1 0>, Passthrough Vector=<-1 0 0>)
    Rotation=&MaskVector(
     Mask Source=1
     Mask Parameter=$Particles.PartitionParticles.Partition
     Mask=0
     Masked Vector=<0.25 0.5 0>
     Passthrough Vector=<-0.25 0.5 0.3>
    )
    Use Rotation Vector=false
   )
   &CameraOffset(Camera Offset Amount=2)
  ]
  particleUpdateScript [
   &ParticleState() {module=ParticleState001}
   &ScaleMeshSize(Scale Factor=&MakeVector(
     X=1
     Y=3
     Z=&FloatFromCurve(FloatCurve=DataInterface() {propertyValues {Curve {keys [
          {interpMode=RCIM_Cubic, leaveTangent=0.0015, tangentMode=RCTM_Break, time=0, value=1}
          {interpMode=RCIM_Cubic, leaveTangent=-0.1745, time=1, value=1.05}
         ]}}})
    )) {module=ScaleMeshSize001}
   &ScaleMeshSize(Scale Factor=&VectorFromFloat(Value=&RandomRangeFloat(Minimum=0.1, Maximum=0.15)))
   &ScaleColor(
    Scale Alpha=&FloatFromCurve(FloatCurve=DataInterface() {propertyValues {Curve {keys [
         {tangentMode=RCTM_User, time=0, value=1}
         {arriveTangent=-0.0405, interpMode=RCIM_Cubic, leaveTangent=-0.0405, tangentMode=RCTM_User, time=0.7009, value=1}
         {arriveTangent=-8.0089, interpMode=RCIM_Cubic, leaveTangent=-8.0089, tangentMode=RCTM_User, time=1, value=0}
        ]}}})
    Scale RGB=&Lerp_Vector(
     A=&VectorFromFloat(Value=&FloatFromCurve(FloatCurve=DataInterface() {propertyValues {Curve {
          defaultValue=1
          keys [
           {interpMode=RCIM_Cubic, tangentMode=RCTM_User, time=0, value=0}
           {arriveTangent=16.9912, interpMode=RCIM_Cubic, leaveTangent=16.9912, tangentMode=RCTM_User, time=1, value=5}
          ]
         }}}))
     B=&VectorFromFloat(Value=&RandomRangeFloat(Minimum=1, Maximum=&RandomRangeFloat(Minimum=5, Maximum=20), Evaluation Type=::NewEnumerator1))
     Alpha=&FloatFromCurve(
      FloatCurve=DataInterface() {propertyValues {Curve {keys [
          {interpMode=RCIM_Cubic, leaveTangent=2.5294, tangentMode=RCTM_Break, time=0, value=0}
          {arriveTangent=-0.2166, interpMode=RCIM_Cubic, tangentMode=RCTM_Break, time=0.4, value=1}
         ]}}}
      Scale Curve=0.5
     )
    )
    ScaleA=true
    ScaleRGB=true
   )
   &DynamicMaterialParameters(
    Index 0 Param 1=&MakeFloatFromVector(Vector=$Particles.Scale, Channel=::NewEnumerator0)
    Index 0 Param 2=&RandomRangeFloat(Minimum=0, Maximum=10)
    Index 0 Param 3=&FloatFromCurve(
     FloatCurve=DataInterface() {propertyValues {Curve {keys [{interpMode=RCIM_Cubic, leaveTangent=8.1085, tangentMode=RCTM_Break, time=0.8568, value=1}, {tangentMode=RCTM_User, time=1, value=0}]}}}
     Scale Curve=0.07
    )
    Index 0 Param 4=0.8
    Index 1 Param 1=&FloatFromCurve(FloatCurve=DataInterface() {propertyValues {Curve {keys [
         {interpMode=RCIM_Cubic, leaveTangent=1.1034, tangentMode=RCTM_Break, time=0, value=0}
         {arriveTangent=2.2399, tangentMode=RCTM_User, time=0.5, value=0.5}
        ]}}})
    Index 1 Param 2=&FloatFromCurve(FloatCurve=DataInterface() {propertyValues {Curve {keys [<at 0.5 0>, {arriveTangent=3.2053, tangentMode=RCTM_User, time=0.9, value=1}]}}})
    Index 1 Param 3=&FloatFromCurve(FloatCurve=DataInterface() {propertyValues {Curve {keys [<at 0 0 RCIM_Cubic>, {arriveTangent=3.4223, interpMode=RCIM_Cubic, leaveTangent=3.4223, tangentMode=RCTM_User, time=1, value=1}]}}})
    Index 1 Param 4=&FloatFromCurve(
     FloatCurve=DataInterface() {propertyValues {Curve {keys [
         {interpMode=RCIM_Cubic, leaveTangent=0.1169, tangentMode=RCTM_Break, time=0, value=0}
         {arriveTangent=2.5708, interpMode=RCIM_Cubic, leaveTangent=2.5708, tangentMode=RCTM_User, time=1, value=1}
        ]}}}
     Scale Curve=&RandomRangeFloat(Minimum=1, Maximum=40)
    )
    Index 2 Param 1=&FloatFromCurve(FloatCurve=DataInterface() {propertyValues {Curve {keys [
         {interpMode=RCIM_Cubic, leaveTangent=-0.0025, tangentMode=RCTM_Break, time=0, value=0.5}
         {arriveTangent=-0.0316, interpMode=RCIM_Cubic, tangentMode=RCTM_Break, time=1, value=-0.2}
        ]}}})
    Index 2 Param 2=&FloatFromCurve(
     FloatCurve=DataInterface() {propertyValues {Curve {keys [
         {interpMode=RCIM_Cubic, tangentMode=RCTM_User, time=0, value=0}
         {arriveTangent=2.0522, interpMode=RCIM_Cubic, leaveTangent=2.0522, tangentMode=RCTM_User, time=1, value=1}
        ]}}}
     Scale Curve=2
    )
    Module.Index0Float0Write=true
    Module.Index0Float1Write=true
    Module.Index0Float2Write=true
    Module.Index0Float3Write=true
    Module.Index1Float0Write=true
    Module.Index1Float1Write=true
    Module.Index1Float2Write=true
    Module.Index1Float3Write=true
    Module.Index2Float0Write=true
    Module.Index2Float1Write=true
    Module.Index2Float2Write=false
    Module.Index2Float3Write=false
    Module.Index3Float0Write=false
    Module.Index3Float1Write=false
    Module.Index3Float2Write=false
    Module.Index3Float3Write=true
    Write Parameter Index 1=::NewEnumerator1
    Write Parameter Index 2=::NewEnumerator1
    Write Parameter Index 3=::NewEnumerator1
   )
  ]
  rendererClasses [&NiagaraMeshRendererProperties]
  renderers [{rendererClass=&NiagaraMeshRendererProperties, rendererIndex=0}]
 }
 Small_Lightning {
  emitterUpdateScript [
   &EmitterState(Life Cycle Mode=::NewEnumerator1, Loop Behavior=::NewEnumerator1, Loop Delay=9.341, Loop Duration=0.39, UseLoopDelay=true)
   &SpawnRate(
    SpawnRate=&FloatFromCurve(
     FloatCurve=DataInterface() {propertyValues {Curve {keys [{interpMode=RCIM_Cubic, tangentMode=RCTM_User, time=0.5, value=1}, {interpMode=RCIM_Cubic, tangentMode=RCTM_User, time=1, value=0.2}]}}}
     CurveIndex=$Emitter.NormalizedLoopAge
     Scale Curve=40
    )
    Use Spawn Probability=false
   )
  ]
  particleSpawnScript [
   &InitializeParticle(
    Color=<rgba 5 3.9929 0 3>
    Lifetime Max=0.055
    Lifetime Min=0.045
    Lifetime Mode=::NewEnumerator1
    Mesh Scale Max=<1 0.5 0.5>
    Mesh Scale Min=<0.25 0.75 0.45>
    Mesh Scale Mode=::NewEnumerator2
    Position Offset=<0 0 20>
    Sprite Size Mode=::NewEnumerator3
    UsePositionOffset=false
    Write Material Random=true
   )
   &CameraOffset(Camera Offset Amount=&RandomRangeFloat(Minimum=20, Maximum=20))
   &InitialMeshOrientation(Mesh Orientation Mode=::NewEnumerator2, Rotation=<0 0.25 0>, Use Rotation Vector=true)
   &InitialMeshOrientation(
    Mesh Orientation Mode=::NewEnumerator2
    Rotation=&RandomRangeVector(Minimum=<0 0 0>, Maximum=<1 0 0>)
    Use Rotation Vector=true
   ) {module=InitialMeshOrientation001}
   &AddVelocity(Module.Use Falloff Distance=false, Override Seed=false, Velocity Mode=::NewEnumerator1, Velocity Speed=1)
   &PartitionParticles(Number Of Partitions=15, Particles Per Partition=1, Partition Mode:ENiagara_PartitionMode_v2=::NewEnumerator1)
  ]
  particleUpdateScript [
   &ParticleState()
   &SkeletalMeshLocation(
    Bone / Socket Index=&SelectIntFromArray(
     Array Sampling Mode=::NewEnumerator1
     Direct Array Index=$Particles.PartitionParticles.Partition
     Int Selection Array=DataInterface() {propertyValues {IntData [0,1,2,3,4,5,6,7,8]}}
     Random Seed=0
    )
    Bone Sampling Mode=::NewEnumerator4
    MaskSpawnGroup=false
    Orientation Sampling=::NewEnumerator1
    Skeletal Mesh=$System.Skeletal Mesh
   ) {module=SkeletalMeshLocation001}
   &JitterPosition(Jitter Amount=0.5)
   &DynamicMaterialParameters(
    Index 0 Param 1=&FloatFromCurve(FloatCurve=DataInterface() {propertyValues {Curve {keys [<at 0 0>,<at 0.7 1>]}}})
    Index 0 Param 2=0.7
    Index 0 Param 3=&FloatFromCurve(FloatCurve=DataInterface() {propertyValues {Curve {keys [
         {arriveTangent=0.2581, interpMode=RCIM_Cubic, leaveTangent=0.2581, tangentMode=RCTM_User, time=0, value=0}
         {arriveTangent=2.8205, interpMode=RCIM_Cubic, leaveTangent=2.8205, tangentMode=RCTM_User, time=1, value=1}
        ]}}})
    Index 0 Param 4=&RandomRangeFloat(Minimum=2, Maximum=3)
    Index 1 Param 1=&FloatFromCurve(FloatCurve=DataInterface() {propertyValues {Curve {keys [<at 0 1>,<at 1 0.3>]}}})
    Module.Index0Float0Write=true
    Module.Index0Float1Write=true
    Module.Index0Float2Write=true
    Module.Index0Float3Write=true
    Module.Index1Float0Write=true
    Module.Index1Float1Write=true
    Module.Index1Float2Write=true
    Module.Index1Float3Write=true
    Write Parameter Index 1=::NewEnumerator1
   ) {module=DynamicMaterialParameters001}
   &SolveForcesAndVelocity(Clamp Velocity=false, Limit Acceleration=false)
  ]
  rendererClasses [&NiagaraMeshRendererProperties]
  renderers [{rendererClass=&NiagaraMeshRendererProperties, rendererIndex=0}]
 }
}
summary {
 systemName=NS_Emote_MorningSoak_Lightning
 emitters [
  {emitterName=HandLightning, rendererClasses [&NiagaraMeshRendererProperties]}
  {emitterName=Flame, rendererClasses []}
  {emitterName=Small_Lightning, rendererClasses [&NiagaraMeshRendererProperties]}
  {emitterName=Fire, rendererClasses [&NiagaraSpriteRendererProperties]}
 ]
 userVariables [
  User.bIsFrontEnd {type=NiagaraBool}
  User.bIsFrontEndPreview {type=NiagaraBool}
 ]
}
