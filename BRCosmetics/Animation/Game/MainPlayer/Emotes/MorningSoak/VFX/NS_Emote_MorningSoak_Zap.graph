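# Niagara effect for the Morning Soak emote that flickers short-lived yellow electric zap meshes on a character's skeleton.
# The only enabled emitter, Electricity_01001, runs two curve-driven spawn groups, using the group for mesh scale and visibility tag across two mesh renderers.
# Particles follow bones selected from an index array, with position jitter and animated dynamic material parameters.
# Two disabled GPU emitters hold single-group versions of the same effect.
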NiagaraSystem /BRCosmetics/Animation/Game/MainPlayer/Emotes/MorningSoak/VFX/NS_Emote_MorningSoak_Zap
dependencies {
 usedDataInterfaces [&NiagaraDataInterfaceCurve, &NiagaraDataInterfaceArrayInt32, &NiagaraDataInterfaceSkeletalMesh]
 usedDynamicInputs []
 usedModules [
  &SystemState
  &self:SystemSpawnScript.SystemScriptSource.SystemScriptGraph.NiagaraNodeAssignment_0.SetVariables_#1
 ]
}
emitters {
 Electricity_01 {
  bEnabled=false
  simTarget=GPUComputeSim
  emitterUpdateScript [
   &EmitterState(Life Cycle Mode=::NewEnumerator1, Loop Behavior=::NewEnumerator1, Loop Delay=0.1, Loop Duration=1.03, UseLoopDelay=false)
   &SpawnRate(
    SpawnRate=&FloatFromCurve(
     FloatCurve=DataInterface() {propertyValues {Curve {keys [{interpMode=RCIM_Cubic, tangentMode=RCTM_User, time=0.5, value=1}, {interpMode=RCIM_Cubic, tangentMode=RCTM_User, time=1, value=0.2}]}}}
     CurveIndex=$Emitter.NormalizedLoopAge
     Scale Curve=30
    )
    Use Spawn Probability=false
   )
  ]
  particleSpawnScript [
   &InitializeParticle(
    Color=<rgba 5 3.9929 0 3>
    Lifetime Max=0.055
    Lifetime Min=0.045
    Lifetime Mode=::NewEnumerator1
    Mesh Scale Max=<1 0.5 0.5>
    Mesh Scale Min=<0.5 1.5 0.9>
    Mesh Scale Mode=::NewEnumerator2
    Position Offset=<0 0 20>
    Sprite Size Mode=::NewEnumerator3
    UsePositionOffset=false
    Write Material Random=true
   )
   /SetVariables(Particles.VisibilityTag=&SpawnGroup()) {isSetParameters=true, module=SetVariables_#2}
   &CameraOffset(Camera Offset Amount=&RandomRangeFloat(Minimum=20, Maximum=20))
   &InitialMeshOrientation(Mesh Orientation Mode=::NewEnumerator2, Rotation=<0 0.25 0>, Use Rotation Vector=true)
   &InitialMeshOrientation(
    Mesh Orientation Mode=::NewEnumerator2
    Rotation=&RandomRangeVector(Minimum=<0 0 0>, Maximum=<1 0 0>)
    Use Rotation Vector=true
   ) {module=InitialMeshOrientation001}
   &AddVelocity(Module.Use Falloff Distance=false, Override Seed=false, Velocity Mode=::NewEnumerator1, Velocity Speed=1)
   &PartitionParticles(Number Of Partitions=15, Particles Per Partition=1, Partition Mode:ENiagara_PartitionMode=::NewEnumerator1)
  ]
  particleUpdateScript [
   &ParticleState()
   &SkeletalMeshLocation(
    Bone / Socket Index=&SelectIntFromArray(
     Array Sampling Mode=::NewEnumerator1
     Direct Array Index=$Particles.PartitionParticles.Partition
     Int Selection Array=DataInterface() {propertyValues {IntData [4,3,14,13,5,1,6,0,2,12,11,8,7,10,9]}}
     Random Seed=0
    )
    Bone Sampling Mode=::NewEnumerator4
    MaskSpawnGroup=false
    Orientation Sampling=::NewEnumerator1
    Skeletal Mesh=$System.Skeletal Mesh
   ) {module=SkeletalMeshLocation001}
   &JitterPosition(Jitter Amount=1)
   &DynamicMaterialParameters(
    Index 0 Param 1=&FloatFromCurve(FloatCurve=DataInterface() {propertyValues {Curve {keys [<at 0 0>,<at 0.7 1>]}}})
    Index 0 Param 2=0.7
    Index 0 Param 3=&FloatFromCurve(FloatCurve=DataInterface() {propertyValues {Curve {keys [
         {arriveTangent=0.2581, interpMode=RCIM_Cubic, leaveTangent=0.2581, tangentMode=RCTM_User, time=0, value=0}
         {arriveTangent=2.8205, interpMode=RCIM_Cubic, leaveTangent=2.8205, tangentMode=RCTM_User, time=1, value=1}
        ]}}})
    Index 0 Param 4=&RandomRangeFloat(Minimum=4, Maximum=6)
    Index 1 Param 1=&FloatFromCurve(FloatCurve=DataInterface() {propertyValues {Curve {keys [<at 0 1>,<at 1 0.3>]}}})
    Module.Index0Float0Write=true
    Module.Index0Float1Write=true
    Module.Index0Float2Write=true
    Module.Index0Float3Write=true
    Module.Index1Float0Write=true
    Module.Index1Float1Write=true
    Module.Index1Float2Write=true
    Module.Index1Float3Write=true
    Write Parameter Index 1=::NewEnumerator1
   ) {module=DynamicMaterialParameters001}
   &SolveForcesAndVelocity(Clamp Velocity=false, Limit Acceleration=false)
  ]
  rendererClasses [&NiagaraMeshRendererProperties]
  renderers [{rendererClass=&NiagaraMeshRendererProperties, rendererIndex=0}]
 }
 Electricity_01001 {
  emitterUpdateScript [
   &EmitterState(Life Cycle Mode=::NewEnumerator1, Loop Behavior=::NewEnumerator1, Loop Delay=0.1, Loop Duration=1.03, UseLoopDelay=false)
   &SpawnRate(
    SpawnRate=&FloatFromCurve(
     FloatCurve=DataInterface() {propertyValues {Curve {keys [{interpMode=RCIM_Cubic, tangentMode=RCTM_User, time=0.5, value=1}, {interpMode=RCIM_Cubic, tangentMode=RCTM_User, time=1, value=0.2}]}}}
     CurveIndex=$Emitter.NormalizedLoopAge
     Scale Curve=30
    )
    Use Spawn Probability=false
   )
   &SpawnRate(
    SpawnGroup=1
    SpawnRate=&FloatFromCurve(
     FloatCurve=DataInterface() {propertyValues {Curve {keys [{interpMode=RCIM_Cubic, tangentMode=RCTM_User, time=0.5, value=1}, {interpMode=RCIM_Cubic, tangentMode=RCTM_User, time=1, value=0.2}]}}}
     CurveIndex=$Emitter.NormalizedLoopAge
     Scale Curve=30
    )
    Use Spawn Probability=false
   ) {module=SpawnRate001}
  ]
  particleSpawnScript [
   /SetVariables(Particles.VisibilityTag=&SpawnGroup()) {isSetParameters=true, module=SetVariables_#2}
   &InitializeParticle(
    Color=<rgba 5 3.9929 0 3>
    Lifetime Max=0.055
    Lifetime Min=0.045
    Lifetime Mode=::NewEnumerator1
    Mesh Scale=&MaskVectorBySpawnGroup(
     Masked Vector=&RandomRangeVector(Minimum=<0.5 1.5 0.9>, Maximum=<1 0.5 0.5>)
     Passthrough Vector=&RandomRangeVector(Minimum=<0.2 0.2 0.2>, Maximum=<0.4 0.4 0.4>)
    )
    Mesh Scale Mode=::NewEnumerator4
    Position Offset=<0 0 20>
    Sprite Size Mode=::NewEnumerator3
    UsePositionOffset=false
    Write Material Random=true
   )
   &CameraOffset(Camera Offset Amount=&RandomRangeFloat(Minimum=20, Maximum=20))
   &InitialMeshOrientation(Mesh Orientation Mode=::NewEnumerator2, Rotation=<0 0.25 0>, Use Rotation Vector=true)
   &InitialMeshOrientation(
    Mesh Orientation Mode=::NewEnumerator2
    Rotation=&RandomRangeVector(Minimum=<0 0 0>, Maximum=<1 0 0>)
    Use Rotation Vector=true
   ) {module=InitialMeshOrientation001}
   &AddVelocity(Module.Use Falloff Distance=false, Override Seed=false, Velocity Mode=::NewEnumerator1, Velocity Speed=1)
  ]
  particleUpdateScript [
   &ParticleState()
   &SkeletalMeshLocation(
    Bone / Socket Index=&SelectIntFromArray(
     Array Sampling Mode=::NewEnumerator1
     Direct Array Index=$Particles.PartitionParticles.Partition
     Int Selection Array=DataInterface() {propertyValues {IntData [4,3,14,13,5,1,6,0,2,12,11,8,7,10,9]}}
     Random Seed=0
    )
    Bone Sampling Mode=::NewEnumerator4
    MaskSpawnGroup=false
    Orientation Sampling=::NewEnumerator1
    Skeletal Mesh=$System.Skeletal Mesh
   ) {module=SkeletalMeshLocation001}
   &JitterPosition(Jitter Amount=1)
   &DynamicMaterialParameters(
    Index 0 Param 1=&FloatFromCurve(FloatCurve=DataInterface() {propertyValues {Curve {keys [<at 0 0>,<at 0.7 1>]}}})
    Index 0 Param 2=0.7
    Index 0 Param 3=&FloatFromCurve(FloatCurve=DataInterface() {propertyValues {Curve {keys [
         {arriveTangent=0.2581, interpMode=RCIM_Cubic, leaveTangent=0.2581, tangentMode=RCTM_User, time=0, value=0}
         {arriveTangent=2.8205, interpMode=RCIM_Cubic, leaveTangent=2.8205, tangentMode=RCTM_User, time=1, value=1}
        ]}}})
    Index 0 Param 4=&RandomRangeFloat(Minimum=4, Maximum=6)
    Index 1 Param 1=&FloatFromCurve(FloatCurve=DataInterface() {propertyValues {Curve {keys [<at 0 1>,<at 1 0.3>]}}})
    Module.Index0Float0Write=true
    Module.Index0Float1Write=true
    Module.Index0Float2Write=true
    Module.Index0Float3Write=true
    Module.Index1Float0Write=true
    Module.Index1Float1Write=true
    Module.Index1Float2Write=true
    Module.Index1Float3Write=true
    Write Parameter Index 1=::NewEnumerator1
   ) {module=DynamicMaterialParameters001}
   &SolveForcesAndVelocity(Clamp Velocity=false, Limit Acceleration=false)
  ]
  rendererClasses [&NiagaraMeshRendererProperties]
  renderers [
   {rendererClass=&NiagaraMeshRendererProperties, rendererIndex=0}
   {rendererClass=&NiagaraMeshRendererProperties, rendererIndex=1}
  ]
 }
 Electricity_02 {
  bEnabled=false
  simTarget=GPUComputeSim
  emitterUpdateScript [
   &EmitterState(Life Cycle Mode=::NewEnumerator1, Loop Behavior=::NewEnumerator1, Loop Delay=0.1, Loop Duration=1.03, UseLoopDelay=false)
   &SpawnRate(
    SpawnRate=&FloatFromCurve(
     FloatCurve=DataInterface() {propertyValues {Curve {keys [{interpMode=RCIM_Cubic, tangentMode=RCTM_User, time=0.5, value=1}, {interpMode=RCIM_Cubic, tangentMode=RCTM_User, time=1, value=0.2}]}}}
     CurveIndex=$Emitter.NormalizedLoopAge
     Scale Curve=30
    )
    Use Spawn Probability=false
   )
  ]
  particleSpawnScript [
   &InitializeParticle(
    Color=<rgba 5 3.9929 0 3>
    Lifetime Max=0.055
    Lifetime Min=0.045
    Lifetime Mode=::NewEnumerator1
    Mesh Scale Max=<0.4 0.4 0.4>
    Mesh Scale Min=<0.2 0.2 0.2>
    Mesh Scale Mode=::NewEnumerator2
    Position Offset=<0 0 20>
    UsePositionOffset=false
    Write Material Random=true
   )
   &CameraOffset(Camera Offset Amount=&RandomRangeFloat(Minimum=20, Maximum=20))
   &InitialMeshOrientation(Mesh Orientation Mode=::NewEnumerator2, Rotation=<0 0.25 0>, Use Rotation Vector=true)
   &InitialMeshOrientation(
    Mesh Orientation Mode=::NewEnumerator2
    Rotation=&RandomRangeVector(Minimum=<0 0 0>, Maximum=<1 0 0>)
    Use Rotation Vector=true
   ) {module=InitialMeshOrientation001}
   &AddVelocity(Module.Use Falloff Distance=false, Override Seed=false, Velocity Mode=::NewEnumerator1, Velocity Speed=1)
   &PartitionParticles(Number Of Partitions=15, Particles Per Partition=1, Partition Mode:ENiagara_PartitionMode=::NewEnumerator1) {enabled=false}
  ]
  particleUpdateScript [
   &ParticleState()
   &SkeletalMeshLocation(
    Bone / Socket Index=&SelectIntFromArray(
     Array Sampling Mode=::NewEnumerator1
     Direct Array Index=$Particles.PartitionParticles.Partition
     Int Selection Array=DataInterface() {propertyValues {IntData [4,3,14,13,5,1,6,0,2,12,11,8,7,10,9]}}
     Random Seed=0
    )
    Bone Sampling Mode=::NewEnumerator4
    MaskSpawnGroup=false
    Orientation Sampling=::NewEnumerator1
    Skeletal Mesh=$System.Skeletal Mesh
   ) {module=SkeletalMeshLocation001}
   &JitterPosition(Jitter Amount=1)
   &DynamicMaterialParameters(
    Index 0 Param 1=&FloatFromCurve(FloatCurve=DataInterface() {propertyValues {Curve {keys [<at 0 0>,<at 0.7 1>]}}})
    Index 0 Param 2=0.7
    Index 0 Param 3=&FloatFromCurve(FloatCurve=DataInterface() {propertyValues {Curve {keys [
         {arriveTangent=0.2581, interpMode=RCIM_Cubic, leaveTangent=0.2581, tangentMode=RCTM_User, time=0, value=0}
         {arriveTangent=2.8205, interpMode=RCIM_Cubic, leaveTangent=2.8205, tangentMode=RCTM_User, time=1, value=1}
        ]}}})
    Index 0 Param 4=&RandomRangeFloat(Minimum=4, Maximum=6)
    Index 1 Param 1=&FloatFromCurve(FloatCurve=DataInterface() {propertyValues {Curve {keys [<at 0 1>,<at 1 0.3>]}}})
    Module.Index0Float0Write=true
    Module.Index0Float1Write=true
    Module.Index0Float2Write=true
    Module.Index0Float3Write=true
    Module.Index1Float0Write=true
    Module.Index1Float1Write=true
    Module.Index1Float2Write=true
    Module.Index1Float3Write=true
    Write Parameter Index 1=::NewEnumerator1
   ) {module=DynamicMaterialParameters001}
   &SolveForcesAndVelocity(Clamp Velocity=false, Limit Acceleration=false)
  ]
  rendererClasses [&NiagaraMeshRendererProperties]
  renderers [{rendererClass=&NiagaraMeshRendererProperties, rendererIndex=0}]
 }
}
summary {
 systemName=NS_Emote_MorningSoak_Zap
 emitters [
  {bEnabled=false, emitterName=Electricity_01, simTarget=GPUComputeSim, rendererClasses [&NiagaraMeshRendererProperties]}
  {bEnabled=false, emitterName=Electricity_02, simTarget=GPUComputeSim, rendererClasses [&NiagaraMeshRendererProperties]}
  {emitterName=Electricity_01001, rendererClasses [&NiagaraMeshRendererProperties]}
 ]
 userVariables [
  User.bIsFrontEnd {type=NiagaraBool}
  User.bIsFrontEndPreview {type=NiagaraBool}
 ]
}
